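# Translucent UI material for the base shape of frontend lobby nameplates, built from a bolt outline and a center circle.
# The bolt MSDF texture is sampled with center-scaled, offset UVs, converted to an SDF and anti-aliased for crisp edges.
# A procedural circle, sized by CircleSize, feeds the CenterCircle reroute.
# The bolt mask is combined with a rerouted mask through Min, and a Color parameter tints through a masked lerp.
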Material /Game/UI/Frontend/Lobby/Nameplates/Materials/M_Nameplate_BaseShape
graph {
 properties {
  bAutomaticallySetUsageInEditor=false
  bUsedWithStaticMesh=false
  blendMode=BLEND_Translucent
  materialDomain=MD_UI
  usedShadingModels=Default Lit
 }
 nodes [
  n1 = Subtract(
   A=MaterialFunctionCall(
    UVs=MaterialFunctionCall() {materialFunction=&GetUserInterfaceUV}.Normalized UV
    Texture Scale=Constant() {r=1.5}
   ) {materialFunction=&ScaleUVsByCenter}.UVs
   B=Constant2Vector() {r=-0.17}
  )
  n2 = MaterialFunctionCall(
   DistanceField=MaterialFunctionCall!(MSDF RGB=TextureSample(UVs=@n1) {samplerType=SAMPLERTYPE_LinearColor, texture=/Game/UI/Frontend/Lobby/Nameplates/Textures/T_UI_ShopLobby_BoltShapeOutline}.RGB) {materialFunction=/FortUILibrary/MaterialFunctions/Shapes/Modifiers/MF_UI_MSDFtoSDF}.SDF
   Threshold=Constant() {r=0.5}
  ) {materialFunction=&MF_UI_AntiAliasing}
  n3 = NamedRerouteUsage()
  LinearInterpolate(Alpha=Multiply(A=Min(A=@n2.Outer, B=@n3))) {constB=0}
  LinearInterpolate(B=VectorParameter() {parameterName=Color, defaultValue=<rgba 0.9301 1 0.0103 1>}.RGB, Alpha=NamedRerouteUsage()) {constA=1}
  n4 = MaterialFunctionCall() {materialFunction=&GetUserInterfaceUV}
  n5 = Constant2Vector()
  n6 = ScalarParameter() {defaultValue=1, group=CenterCircle, parameterName=CircleSize}
  NamedRerouteDeclaration(None=MaterialFunctionCall(UVs=Add(A=@n4.Normalized UV, B=@n5), Size=LinearInterpolate(Alpha=@n6)) {materialFunction=&MF_UI_Circle}.Fill) {name=CenterCircle, nodeColor=<rgba 1 0.3281 0 1>}
  NamedRerouteUsage()
 ]
}
