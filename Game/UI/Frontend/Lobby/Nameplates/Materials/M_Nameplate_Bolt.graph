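# Translucent UI material for a lobby nameplate bolt graphic, with parameters for bolt animation, burst animation and glow.
# An MSDF bolt and a half circle cut from two SDF circles use offset UVs, alongside panning noise for distortion.
# A radial progress mask clips combined shapes through Min and Max.
# One progress parameter scales the MSDF burst texture and shifts its threshold, while a radial gradient forms the glow.
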Material /Game/UI/Frontend/Lobby/Nameplates/Materials/M_Nameplate_Bolt
graph {
 properties {
  bAutomaticallySetUsageInEditor=false
  bUsedWithStaticMesh=false
  blendMode=BLEND_Translucent
  materialDomain=MD_UI
  usedShadingModels=Default Lit
 }
 nodes [
  Constant2Vector() {r=-0.08}
  Constant2Vector() {r=-0.043}
  Constant2Vector() {r=0.073}
  Constant2Vector() {r=0.07}
  Constant3Vector() {constant=<rgba 1 1 1 0>}
  Max(A=NamedRerouteUsage(), B=NamedRerouteUsage())
  n1 = NamedRerouteUsage()
  n2 = NamedRerouteUsage()
  n3 = MaterialFunctionCall!(
   UVs
   Progress=ScalarParameter() {defaultValue=1, parameterName=BoltAnimationProgress}
   StartValue=ScalarParameter() {defaultValue=220, group=BoltAnimation, parameterName=StartValue, sortPriority=40}
   Range=ScalarParameter() {defaultValue=360, group=BoltAnimation, parameterName=Range}
  ) {materialFunction=/FortUILibrary/MaterialFunctions/Basics/MF_UI_RadialProgress}
  NamedRerouteDeclaration(None=Min(A=Max(A=@n1, B=@n2), B=NamedRerouteDeclaration(None=@n3.Result) {name=RadialAnimation, nodeColor=<rgba 1 0.9609 0 1>})) {name=BoltAnimation, nodeColor=<rgba 0 0.7891 1 1>}
  n4 = LinearInterpolate(Alpha=ScalarParameter() {defaultValue=0.5, parameterName=BurstAnimationProgress}) {constB=0.25}
  n5 = TextureSample(UVs=MaterialFunctionCall(Texture Scale=LinearInterpolate(Alpha=@n4) {constB=24}) {materialFunction=&ScaleUVsByCenter}.UVs) {samplerType=SAMPLERTYPE_LinearColor, texture=/Game/UI/Frontend/Lobby/Nameplates/Textures/T_UI_ShopLobby_LargeBurst}
  NamedRerouteDeclaration(None=MaterialFunctionCall(
    DistanceField=MaterialFunctionCall!(MSDF RGB=@n5.RGB) {materialFunction=/FortUILibrary/MaterialFunctions/Shapes/Modifiers/MF_UI_MSDFtoSDF}.SDF
    Threshold=LinearInterpolate(Alpha=@n4) {constA=0.4}
   ) {materialFunction=&MF_UI_AntiAliasing}.Outer) {name=Animation, nodeColor=<rgba 0 1 0.5312 1>}
  n6 = MaterialFunctionCall(
   UVs=MaterialFunctionCall() {materialFunction=&GetUserInterfaceUV}.Normalized UV
   Texture Scale=ScalarParameter() {defaultValue=1, group=Glow, parameterName=GlowSize}
  ) {materialFunction=&ScaleUVsByCenter}
  n7 = ScalarParameter() {group=Glow, parameterName=GlowMin}
  n8 = ScalarParameter() {defaultValue=1, group=Glow, parameterName=GlowMax}
  NamedRerouteDeclaration(None=Multiply(
    A=MaterialFunctionCall(UV=@n6.UVs, GradientStart=@n7, GradientEnd=@n8) {materialFunction=&MF_UI_RadialGradient}.Smooth
    B=ScalarParameter() {defaultValue=1, group=Glow, parameterName=GlowIntensity}
   )) {name=Glow, nodeColor=<rgba 1 0 0.3047 1>}
  n9 = Add(A=MaterialFunctionCall() {materialFunction=&GetUserInterfaceUV}.Normalized UV, B=NamedRerouteUsage())
  n10 = Constant() {r=0.7}
  n11 = TextureSample(UVs=Add(
    A=MaterialFunctionCall(UVs=@n9, Texture Scale=@n10) {materialFunction=&ScaleUVsByCenter}.UVs
    B=Constant2Vector() {g=0.07, r=0.23}
   )) {samplerType=SAMPLERTYPE_LinearColor, texture=/Game/UI/Frontend/Lobby/Nameplates/Textures/T_UI_ShopLobby_BoltShape}
  NamedRerouteDeclaration(None=MaterialFunctionCall(
    DistanceField=MaterialFunctionCall(MSDF RGB=@n11.RGB) {materialFunction=&MF_UI_MSDFtoSDF}.SDF
    Threshold=Constant() {r=0.57}
   ) {materialFunction=&MF_UI_AntiAliasing}.Outer) {name=BoltShape, nodeColor=<rgba 0 1 0.3672 1>}
  n12 = Panner() {speedX=0.1, speedY=0.1}
  NamedRerouteDeclaration(None=Multiply(
    A=TextureSample(UVs=@n12) {samplerType=SAMPLERTYPE_LinearGrayscale, texture=/Game/Global/Textures/Default/Noise/T_WhiteNoise}.R
    B=ScalarParameter() {group=BoltAnimation, parameterName=BoltDistortion, sliderMax=2}
   )) {name=Noise, nodeColor=<rgba 0 0.2734 1 1>}
  n13 = MaterialFunctionCall() {materialFunction=&GetUserInterfaceUV}
  n14 = NamedRerouteUsage()
  n15 = Add(
   A=MaterialFunctionCall(UVs=Add(A=@n13.Normalized UV, B=@n14), Texture Scale=Constant() {r=0.9}) {materialFunction=&ScaleUVsByCenter}.UVs
   B=Constant2Vector() {r=-0.075}
  )
  n16 = MaterialFunctionCall(UVs=@n15, Size=Constant() {r=0.9}) {materialFunction=&MF_UI_Circle}
  n17 = MaterialFunctionCall(UVs=Add(A=@n15, B=Constant2Vector() {r=0.075}), Size=Constant() {r=0.88}) {materialFunction=&MF_UI_Circle}
  NamedRerouteDeclaration(None=MaterialFunctionCall(DistanceField=MaterialFunctionCall!(DistanceFieldA=@n16.SDF, DistanceFieldB=@n17.SDF, BlendAmount) {materialFunction=/FortUILibrary/MaterialFunctions/Shapes/Modifiers/MF_UI_SDF_Subtraction}.Exact) {materialFunction=&MF_UI_AntiAliasing}.Inner) {name=HalfCircleShape, nodeColor=<rgba 0.5938 1 0 1>}
 ]
}
