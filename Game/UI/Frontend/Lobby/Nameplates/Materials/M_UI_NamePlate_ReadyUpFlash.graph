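# UI material for a green ready-up flash on lobby nameplates, using alpha-composite blending.
# Procedural Flare, Inner Ring and Glow masks come from scaled UVs, radial gradient and SDF circle functions, without textures.
# The masks are blended, saturated, tinted fixed green and multiplied by vertex alpha.
# Scalar parameters drive beam, ring and glow shape; the Time-based ping-pong and easing nodes do not feed these masks.
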Material /Game/UI/Frontend/Lobby/Nameplates/Materials/M_UI_NamePlate_ReadyUpFlash
graph {
 properties {
  bAutomaticallySetUsageInEditor=false
  bUsedWithStaticMesh=false
  blendMode=BLEND_AlphaComposite
  materialDomain=MD_UI
  usedShadingModels=Default Lit
 }
 nodes [
  Frac(None=Time())
  Frac(None=Time())
  LinearInterpolate(B=ScalarParameter() {defaultValue=0.2, parameterName=Beam Height, sliderMax=0.1}) {constAlpha=1}
  LinearInterpolate(A=ScalarParameter() {defaultValue=1.5, parameterName=Beam Width}) {constAlpha=1, constB=0}
  MaterialFunctionCall!(Value=Time(), Scale=ScalarParameter() {defaultValue=0.6, parameterName=Speed}, Height) {materialFunction=/FortUILibrary/MaterialFunctions/Basics/MF_UI_PingPong}
  n1 = ScalarParameter() {defaultValue=0.6, parameterName=Ring -Speed, sliderMax=1}
  n2 = Multiply(A=Time(), B=@n1)
  MaterialFunctionCall("Value (0-1)"=Frac(None=Add(A=@n2) {constB=0.5})) {materialFunction=&MF_UI_EaseCurves}
  n3 = VertexColor()
  n4 = LinearInterpolate(A=NamedRerouteUsage(), Alpha=NamedRerouteUsage())
  n5 = NamedRerouteUsage()
  n6 = Saturate(None=Divide(A=LinearInterpolate(A=@n4, Alpha=@n5)) {constB=0.5})
  Multiply(A=@n3.A, B=Multiply(A=Constant3Vector() {constant=<rgba 0 0.7865 0.0979 1>}, B=@n6))
  Multiply(A=@n3.A, B=@n6)
  n7 = MaterialFunctionCall() {materialFunction=&GetUserInterfaceUV}
  n8 = Multiply(A=Abs(None=Subtract(A=@n7.Normalized UV) {constB=0.5})) {constB=2}
  n9 = ComponentMask(None=MaterialFunctionCall(UVs=@n8) {materialFunction=&ScaleUVsByCenter}.UVs) {g=true}
  n10 = MaterialFunctionCall(UVs=@n8) {materialFunction=&ScaleUVsByCenter}
  n11 = Multiply(A=LinearInterpolate(A=@n9, Alpha=ComponentMask(None=@n10.UVs) {r=true}), B=@n9)
  n12 = Saturate(None=Power(Base=OneMinus(None=@n11), Exp=Constant() {r=60}))
  n13 = AppendVector(
   A=ScalarParameter() {defaultValue=1.5, parameterName=Beam Width}
   B=ScalarParameter() {defaultValue=0.2, parameterName=Beam Height, sliderMax=0.1}
  )
  n14 = LinearInterpolate(Alpha=MaterialFunctionCall(UV=MaterialFunctionCall(Texture Scale=@n13) {materialFunction=&ScaleUVsByCenter}.UVs) {materialFunction=&MF_UI_RadialGradient}.Smooth)
  NamedRerouteDeclaration(None=Multiply(A=ScalarParameter() {defaultValue=1, parameterName=Beam Opacity}, B=Multiply(A=@n12, B=@n14))) {name=Flare, nodeColor=<rgba 0 0 0 1>}
  n15 = MaterialFunctionCall(UV=MaterialFunctionCall() {materialFunction=&GetUserInterfaceUV}.Normalized UV) {materialFunction=&MF_UI_UVScaling}
  n16 = LinearInterpolate(A=ScalarParameter() {parameterName=Ring Min Size}, B=ScalarParameter() {defaultValue=0.8, parameterName=Ring Max Size}, Alpha=@n1)
  n17 = Constant() {r=0}
  n18 = Constant() {r=0.025}
  n19 = ScalarParameter() {defaultValue=0.05, parameterName=Ring Thickness}
  n20 = MaterialFunctionCall!(
   In=Subtract(
    A=MaterialFunctionCall(UVs=@n15.Scaled to Fit, Size=@n16, OutlineThickness=@n17, GlowMax=@n18) {materialFunction=&MF_UI_Circle}.SDFOutline
    B=LinearInterpolate(A=@n19, Alpha=@n1) {constA=0.025, constB=0}
   ) {constB=0.005}
   Ramp Width=ScalarParameter() {defaultValue=0.1, parameterName=Ring Smoothness}
  ) {materialFunction=/Engine/Functions/Engine_MaterialFunctions01/Math/SmoothCeil}
  NamedRerouteDeclaration(None=Saturate(None=OneMinus(None=@n20.Result))) {name=Inner Ring Mask, nodeColor=<rgba 0 0 0 1>}
  n21 = MaterialFunctionCall(UV=MaterialFunctionCall() {materialFunction=&GetUserInterfaceUV}.Normalized UV) {materialFunction=&MF_UI_UVScaling}
  n22 = ScalarParameter() {defaultValue=1, parameterName=Glow Strength/Opacity, sliderMax=1}
  n23 = LinearInterpolate(B=ScalarParameter() {defaultValue=0.1, parameterName=Glow Max Size}, Alpha=@n22)
  n24 = LinearInterpolate(B=ScalarParameter() {defaultValue=0.3, parameterName=Glow Radius}, Alpha=@n22)
  NamedRerouteDeclaration(None=MaterialFunctionCall("Value (0-1)"=MaterialFunctionCall(UVs=@n21.Scaled to Fit, Size=@n23, GlowMax=@n24) {materialFunction=&MF_UI_Circle}.Fill) {materialFunction=&MF_UI_EaseCurves}.EaseInCubic) {name=Glow Mask, nodeColor=<rgba 0 0 0 1>}
  n25 = ScalarParameter() {defaultValue=1.5, parameterName=Beam Width}
  n26 = MaterialFunctionCall() {materialFunction=&MF_UI_EaseCurves}
  n27 = ScalarParameter() {defaultValue=0.2, parameterName=Beam Height, sliderMax=0.1}
  OneMinus(None=AppendVector(
    A=LinearInterpolate(A=@n25, Alpha=@n26.EaseInSine) {constAlpha=1, constB=0}
    B=LinearInterpolate(B=@n27, Alpha=@n26.EaseInSine) {constAlpha=1}
   ))
 ]
}
