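# UI material function that draws a hexagon with adjustable radius, corner radius and pixel-based outline thickness from a signed distance field.
# Remaps UVs to -1..1, folds them with abs and a dot-product reflection against a constant helper vector, then clamps to an edge for signed distance.
# Outputs antialiased Fill and inset Outline masks, the raw SDF, and an absolute-valued outline SDF.
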MaterialFunction /FortUILibrary/MaterialFunctions/Shapes/MF_UI_HexagonSDF
graph {nodes [
  n1 = MaterialFunctionCall!(
   Value PX=FunctionInput(Preview=Constant() {r=1}) {inputName=Outline Thickness, inputType=FunctionInput_Scalar, sortPriority=3}
   UV Pixel Size
   X or Y
  ) {materialFunction=/FortUILibrary/MaterialFunctions/Basics/MF_UI_PixelsToNormalized}
  n2 = Abs(None=NamedRerouteUsage())
  n3 = NamedRerouteUsage()
  n4 = Min(A=DotProduct(A=ComponentMask(None=@n3) {g=true, r=true}, B=@n2)) {constB=0}
  n5 = NamedRerouteUsage()
  n6 = Subtract(A=@n2, B=Multiply(A=Multiply(A=@n4) {constB=2}, B=ComponentMask(None=@n5) {g=true, r=true}))
  n7 = ComponentMask(None=@n6) {r=true}
  n8 = Multiply(A=ComponentMask(None=NamedRerouteUsage()) {b=true}, B=NamedRerouteUsage())
  n9 = Multiply(A=@n8) {constB=-1}
  n10 = Subtract(A=@n6, B=AppendVector(A=Clamp(None=@n7, Min=@n9, Max=@n8), B=NamedRerouteUsage()))
  n11 = Length(None=@n10) {is=class MaterialExpressionLength}
  n12 = Sign(None=ComponentMask(None=@n10) {g=true})
  n13 = FunctionOutput(None=Subtract(A=Multiply(A=@n11, B=@n12), B=NamedRerouteUsage())) {outputName=SDF, sortPriority=2}
  n14 = Subtract(A=Abs(None=Add(A=@n13, B=@n1."Normalized (>1px)")), B=@n1."Normalized (>1px)") {constB=0.1}
  FunctionOutput(None=MaterialFunctionCall(DistanceField=@n14) {materialFunction=&MF_UI_AntiAliasing}.Inner) {outputName=Outline, sortPriority=1}
  FunctionOutput(None=MaterialFunctionCall(DistanceField=@n13) {materialFunction=&MF_UI_AntiAliasing}.Inner) {outputName=Fill}
  FunctionOutput(None=Abs(None=@n14)) {outputName=SDF Outline, sortPriority=3}
  n15 = Constant() {r=0.9}
  NamedRerouteDeclaration(None=Subtract(A=FunctionInput(Preview=@n15) {inputName=Radius, inputType=FunctionInput_Scalar, sortPriority=1}, B=NamedRerouteUsage())) {name=radius, nodeColor=<rgba 0 0.3203 1 1>}
  n16 = MaterialFunctionCall() {materialFunction=&GetUserInterfaceUV}
  NamedRerouteDeclaration(None=ConstantBiasScale(None=FunctionInput(Preview=@n16.Normalized UV) {inputName=UV, inputType=FunctionInput_Vector2}) {bias=-0.5, desc=remap coord space from 0 1 to -1 1, scale=2}) {name=UV, nodeColor=<rgba 0.8047 0 1 1>}
  n17 = Multiply(A=SquareRoot(None=Constant() {r=3})) {constB=0.5}
  n18 = Constant() {r=0.1}
  NamedRerouteDeclaration(None=Multiply(A=Multiply(A=@n15, B=@n17), B=FunctionInput(Preview=@n18) {inputName=Corner Radius, inputType=FunctionInput_Scalar, sortPriority=2})) {name=corner radius, nodeColor=<rgba 0.3359 1 0 1>}
  n19 = SquareRoot(None=Constant() {r=3})
  NamedRerouteDeclaration(None=MaterialFunctionCall(R=Multiply(A=@n19) {constB=-0.5}, G=Constant() {r=0.5}, B=Divide(B=@n19)) {materialFunction=&AppendMany}.RGB) {name=helper vector, nodeColor=<rgba 0 1 0.9531 1>}
 ]}
