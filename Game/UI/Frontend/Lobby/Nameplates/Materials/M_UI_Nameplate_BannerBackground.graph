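# UI material for the lobby nameplate banner background, drawing a tinted hexagon with an outline.
# Center-scaled UVs, warped by mirrored vertical position and rotated a quarter turn, feed the hexagon SDF function.
# Its fill and outline reroutes feed the shading of the Banner Color parameter through lerps and MultiplyAdd.
# A static switch in the Ranked group adds an offset circular bottom glow with adjustable size and intensity.
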Material /Game/UI/Frontend/Lobby/Nameplates/Materials/M_UI_Nameplate_BannerBackground
graph {
 properties {
  bUsedWithStaticMesh=false
  blendMode=BLEND_Translucent
  materialDomain=MD_UI
  pixelDepthOffsetMode=PDOM_Legacy
  usedShadingModels=Default Lit
 }
 nodes [
  n1 = MaterialFunctionCall!(
   UV=NamedRerouteUsage()
   Radius=Constant() {r=0.9}
   Corner Radius=ScalarParameter() {defaultValue=0.1, parameterName=Corner Radius}
   Outline Thickness=ScalarParameter() {defaultValue=5, parameterName=Outline Thickness}
  ) {materialFunction=/FortUILibrary/MaterialFunctions/Shapes/MF_UI_HexagonSDF}
  NamedRerouteDeclaration(None=@n1.Fill) {name=hexagonFill, nodeColor=<rgba 0.5234 0 1 1>}
  NamedRerouteDeclaration(None=@n1.Outline) {name=hexagonOutline, nodeColor=<rgba 0 0.5547 1 1>}
  n2 = MaterialFunctionCall() {materialFunction=&GetUserInterfaceUV}
  n3 = ComponentMask(None=@n2.Normalized UV) {g=true}
  n4 = Min(A=OneMinus(None=@n3), B=@n3)
  n5 = Multiply(
   A=LinearInterpolate(A=Constant2Vector() {g=0.95, r=1}, B=Constant2Vector() {g=1.05, r=1}, Alpha=Multiply(A=@n4) {constB=2}) {constA=1}
   B=Constant2Vector() {g=1.025, r=1.1}
  )
  NamedRerouteDeclaration(None=Rotator(
    Coordinate=MaterialFunctionCall(UVs=@n2.Normalized UV, Texture Scale=@n5) {materialFunction=&ScaleUVsByCenter}.UVs
    Time=Constant() {r=0.25}
   ) {speed=6.2832}) {name=hexagonUV, nodeColor=<rgba 1 0.7031 0 1>}
  NamedRerouteUsage()
  n6 = Multiply(A=NamedRerouteUsage(), B=VectorParameter() {parameterName=Banner Color, defaultValue=<rgba 0.0078 0.0667 0.8784 1>}.RGB)
  n7 = Multiply(A=NamedRerouteUsage()) {constB=0.85}
  n8 = LinearInterpolate(
   A=@n6
   B=MaterialFunctionCall!(Base=@n6, Add=LinearInterpolate(Alpha=@n7) {constA=-1}) {materialFunction=/Engine/Functions/Engine_MaterialFunctions02/Math/MultiplyAdd}.Result
   Alpha=NamedRerouteUsage()
  )
  n9 = LinearInterpolate(A=Multiply(A=@n8) {constB=0.5}, B=@n8, Alpha=ComponentMask(None=NamedRerouteUsage()) {r=true})
  n10 = Add(
   A=MaterialFunctionCall() {materialFunction=&GetUserInterfaceUV}.Normalized UV
   B=AppendVector(A=ScalarParameter() {parameterName=X Offset}, B=ScalarParameter() {defaultValue=-0.7, parameterName=Y Offset})
  )
  n11 = ScalarParameter() {defaultValue=0.25, group=Ranked, parameterName=Glow Size}
  n12 = Constant() {r=1.5}
  n13 = Multiply(
   A=Saturate(None=MaterialFunctionCall(UVs=@n10, Size=@n11, GlowMax=@n12) {materialFunction=&MF_UI_Circle}.Fill)
   B=ScalarParameter() {defaultValue=2, group=Ranked, parameterName=Glow Intensity}
  ) {constB=4}
  n14 = LinearInterpolate(Alpha=@n13) {constA=-1}
  StaticSwitchParameter(
   True=LinearInterpolate(
    A=@n9
    B=MaterialFunctionCall(Base=@n9, Add=@n14) {materialFunction=&MultiplyAdd}.Result
    Alpha=@n13
   )
   False=@n9
  ) {group=Ranked, parameterName=Bottom Glow}
 ]
}
