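# Translucent UI material for a lobby nameplate voice chat indicator, with a green color constant and a hexagon outline that responds to a Volume parameter.
# MF_UI_HexagonSDF draws the outline, which is masked by a vertical gradient whose start moves from bottom to top as Volume rises.
# Hexagon UVs are scaled about center, varying with distance from the top and bottom edges, then turned a quarter turn by a constant-time Rotator.
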Material /Game/UI/Frontend/Lobby/Nameplates/Materials/M_UI_Nameplate_VoiceChatOutline
graph {
 properties {
  bUsedWithStaticMesh=false
  blendMode=BLEND_Translucent
  materialDomain=MD_UI
  pixelDepthOffsetMode=PDOM_Legacy
  usedShadingModels=Default Lit
 }
 nodes [
  Constant3Vector() {constant=<rgba 0.0395 1 0.2307 1>}
  n1 = NamedRerouteUsage()
  n2 = Constant() {r=0.9}
  n3 = Constant() {r=0.1}
  n4 = ScalarParameter() {defaultValue=5, parameterName=Outline Thickness}
  n5 = ComponentMask(None=MaterialFunctionCall() {materialFunction=&GetUserInterfaceUV}.Normalized UV) {g=true}
  n6 = OneMinus(None=ScalarParameter() {defaultValue=1, parameterName=Volume, sliderMax=1})
  n7 = Constant() {r=1}
  LinearInterpolate(
   B=NamedRerouteDeclaration(None=MaterialFunctionCall!(UV=@n1, Radius=@n2, Corner Radius=@n3, Outline Thickness=@n4) {materialFunction=/FortUILibrary/MaterialFunctions/Shapes/MF_UI_HexagonSDF}.Outline) {name=hexagonOutline, nodeColor=<rgba 0 0 0 1>}
   Alpha=Saturate(None=MaterialFunctionCall(Input=@n5, Input Low=@n6, Input High=@n7) {materialFunction=&RemapValue_0-1}.Result)
  )
  n8 = MaterialFunctionCall() {materialFunction=&GetUserInterfaceUV}
  n9 = ComponentMask(None=@n8.Normalized UV) {g=true}
  n10 = Min(A=OneMinus(None=@n9), B=@n9)
  n11 = Multiply(
   A=LinearInterpolate(A=Constant2Vector() {g=0.95, r=1}, B=Constant2Vector() {g=1.05, r=1}, Alpha=Multiply(A=@n10) {constB=2}) {constA=1}
   B=Constant2Vector() {g=1.025, r=1.1}
  )
  NamedRerouteDeclaration(None=Rotator(
    Coordinate=MaterialFunctionCall(UVs=@n8.Normalized UV, Texture Scale=@n11) {materialFunction=&ScaleUVsByCenter}.UVs
    Time=Constant() {r=0.25}
   ) {speed=6.2832}) {name=hexagonUV, nodeColor=<rgba 0 0 0 1>}
 ]
}
