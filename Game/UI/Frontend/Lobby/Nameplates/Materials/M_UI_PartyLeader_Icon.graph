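# UI material that draws the party leader icon for lobby nameplates from a signed distance field texture.
# Transformed UVs sample the icon, whose red and green channels become antialiased fill and optional outline masks driving color and opacity.
# The color is premultiplied for alpha-composite blending with brightness and additive controls.
# An optional tint and a color-dodged shine swipe finish the color, scaled by vertex alpha.
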Material /Game/UI/Frontend/Lobby/Nameplates/Materials/M_UI_PartyLeader_Icon
graph {
 properties {
  bUsedWithStaticMesh=false
  blendMode=BLEND_AlphaComposite
  materialDomain=MD_UI
  pixelDepthOffsetMode=PDOM_Legacy
  usedShadingModels=Default Lit
 }
 nodes [
  n1 = NamedRerouteUsage()
  n2 = Multiply(
   A=NamedRerouteUsage()
   B=Multiply(A=ScalarParameter() {defaultValue=1, group=Color, parameterName=Brightness Percentage, sliderMax=1, sortPriority=2}, B=@n1)
  )
  n3 = VectorParameter() {parameterName=Tint Color, defaultValue=<rgba 1 0 0.9925 0>}
  n4 = ScalarParameter() {parameterName="Tint Icon (False)", sliderMax=1}
  Multiply(
   A=VertexColor().A
   B=MaterialFunctionCall!(
    Blend=MaterialFunctionCall!(In_Uv) {materialFunction=/Game/UI/Foundation/Materials/M-Function-ShineSwipe}.Result
    Base=LinearInterpolate(A=@n2, B=@n3.RGB, Alpha=@n4)
   ) {materialFunction=/Engine/Functions/Engine_MaterialFunctions03/Blends/Blend_ColorDodge}.Result
  )
  Multiply(A=@n1, B=OneMinus(None=ScalarParameter() {group=Color, parameterName=Additive Percentage, sliderMax=1, sortPriority=3}))
  n5 = TextureSampleParameter2D(UVs=NamedRerouteUsage()) {
   parameterName=Icon
   samplerSource=SSM_Clamp_WorldGroupSettings
   samplerType=SAMPLERTYPE_LinearColor
   texture=/FortUILibrary/Textures/Icons/T_UI_Icon_UIKit_Leader_SDF
  }
  NamedRerouteDeclaration(None=AppendVector(A=@n5.RGB, B=Constant())) {name=Texture Channels, nodeColor=<rgba 0 0 0 1>}
  n6 = ScalarParameter() {defaultValue=0.25, group=3 - Outline Settings, parameterName=Outline Opacity, sliderMax=1, sortPriority=34}
  n7 = NamedRerouteUsage()
  NamedRerouteDeclaration(None=LinearInterpolate(A=Multiply(A=@n6, B=@n7), B=ScalarParameter() {defaultValue=1, parameterName=Fill Opacity, sliderMax=1}, Alpha=NamedRerouteUsage()) {constAlpha=1}) {name=Opacity, nodeColor=<rgba 0 0 0 1>}
  n8 = VectorParameter() {group=3 - Outline Settings, parameterName=Outline Color, sortPriority=33, defaultValue=<rgba 1 0.8111 0 1>}
  n9 = Multiply(A=Constant() {r=1}, B=NamedRerouteUsage())
  n10 = Constant() {r=1}
  n11 = NamedRerouteUsage()
  NamedRerouteDeclaration(None=LinearInterpolate(
    A=LinearInterpolate(B=@n8.RGB, Alpha=@n9)
    B=VectorParameter() {group=Color, parameterName=Fill Color, defaultValue=<rgba 1 0.6724 0.0612 1>}.RGB
    Alpha=Multiply(A=@n10, B=@n11)
   )) {name=Separated Colors, nodeColor=<rgba 0 0 0 1>}
  n12 = NamedRerouteUsage()
  n13 = ComponentMask(None=@n12) {g=true}
  n14 = OneMinus(None=ScalarParameter() {defaultValue=0.35, group=3 - Outline Settings, parameterName=Outline Scale, sliderMax=1})
  n15 = AppendVector(
   A=MaterialFunctionCall(DistanceField=ComponentMask(None=@n12) {r=true}, Threshold=ScalarParameter() {defaultValue=0.68, parameterName=Fill Aliasing}) {materialFunction=&MF_UI_AntiAliasing}.Outer
   B=StaticSwitchParameter(
    True=MaterialFunctionCall(DistanceField=@n13, Threshold=@n14) {materialFunction=&MF_UI_AntiAliasing}.Outer
    False=Constant()
   ) {group=3 - Outline Settings, parameterName=Display Outline}
  )
  n16 = NamedRerouteUsage()
  n17 = MaterialFunctionCall(Float2=NamedRerouteDeclaration(None=StaticSwitchParameter(True=@n15, False=@n16) {defaultValue=true, group=2 - Distance Field, parameterName=Use Distance Field, sortPriority=0}) {name=Distance Field Pass, nodeColor=<rgba 0 0 0 1>}) {materialFunction=&BreakOutFloat2Components}
  NamedRerouteDeclaration(None=@n17.R) {name=Red Channel, nodeColor=<rgba 0 0 0 1>}
  NamedRerouteDeclaration(None=@n17.G) {name=Green Channel, nodeColor=<rgba 0 0 0 1>}
  n18 = MaterialFunctionCall() {materialFunction=&GetUserInterfaceUV}
  n19 = AppendVector(
   A=Subtract(A=ComponentMask(None=@n18.Normalized UV) {r=true}, B=ScalarParameter() {group=1 - Transform, parameterName=OffsetX, sortPriority=48})
   B=Add(A=ComponentMask(None=@n18.Normalized UV) {g=true}, B=ScalarParameter() {group=1 - Transform, parameterName=OffsetY, sortPriority=48})
  )
  n20 = ScalarParameter() {defaultValue=2, group=1 - Transform, parameterName=Scale, sortPriority=0}
  NamedRerouteDeclaration(None=MaterialFunctionCall(
    UVs=MaterialFunctionCall(UVs=@n19, Texture Scale=@n20) {materialFunction=&ScaleUVsByCenter}.UVs
    "Rotation Angle (0-1)"=ScalarParameter() {group=1 - Transform, parameterName=Rotation}
   ) {materialFunction=&CustomRotator}.Rotated Values) {name=UVs, nodeColor=<rgba 0 0 0 1>}
 ]
}
