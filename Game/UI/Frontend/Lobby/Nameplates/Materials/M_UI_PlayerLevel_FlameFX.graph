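# UI material that draws an animated, tintable flame effect for the player level display on lobby nameplates, using alpha composite blending.
# Panning noise distorts aspect-corrected UVs fed into MF_UI_Circle, bending a circle SDF into a flame with adjustable size and height.
# Smoothsteps build inner glow, wobbly stroke and outer glow layers.
# Curve atlas channels sampled along vertical UV shape distortion and fades, while X/Y offsets vary the pattern between multiple flames.
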Material /Game/UI/Frontend/Lobby/Nameplates/Materials/M_UI_PlayerLevel_FlameFX
graph {
 properties {
  bUsedWithStaticMesh=false
  blendMode=BLEND_AlphaComposite
  materialDomain=MD_UI
  pixelDepthOffsetMode=PDOM_Legacy
  usedShadingModels=Default Lit
 }
 nodes [
  n1 = ScalarParameter() {
   defaultValue=-0.2
   desc="How far in from the flame the inner glow should extend (or end if InnerGlowSoftness is positive)."
   group=Inner Glow
   parameterName=InnerGlowEdge
   sliderMin=-1
   sortPriority=1
  }
  n2 = ScalarParameter() {
   defaultValue=-0.3
   desc="How feathered the edge of the inner glow should be. Positive values will glow from center, negative values will glow from the edge,"
   group=Inner Glow
   parameterName=InnerGlowSoftness
   sliderMax=1
   sliderMin=-1
   sortPriority=2
  }
  n3 = SmoothStep(Min=@n1, Max=Subtract(A=@n1, B=@n2), Value=NamedRerouteUsage()) {constMax=0}
  n4 = ScalarParameter() {defaultValue=1, desc=The opacity of the stroke., group=Stroke, parameterName=StrokeOpacity, sliderMax=1, sortPriority=0}
  n5 = Max(A=@n4) {constB=0}
  n6 = ScalarParameter() {
   defaultValue=0.02
   desc=How feathered the edge of the stroke should be.
   group=Stroke
   parameterName=StrokeSoftness
   sliderMax=1
   sortPriority=2
  }
  n7 = ScalarParameter() {defaultValue=0.03, desc=How thick the stroke should be., group=Stroke, parameterName=StrokeThickness, sliderMax=1, sortPriority=1}
  n8 = ScalarParameter() {
   defaultValue=0.05
   desc=How wobbly the stroke should be as the flame animates.
   group=Stroke
   parameterName=StrokeVariance
   sliderMax=1
   sortPriority=3
  }
  n9 = Add(A=ComponentMask(None=NamedRerouteUsage()) {r=true}) {constB=-0.5}
  n10 = SmoothStep(Min=@n7, Max=Subtract(A=@n7, B=Max(A=@n6) {constB=0}), Value=Add(A=Multiply(A=@n8, B=@n9), B=NamedRerouteUsage())) {constMax=0, constMin=0.05}
  n11 = NamedRerouteUsage()
  n12 = CurveAtlasRowParameter(CurveTime=OneMinus(None=ComponentMask(None=@n11) {g=true})) {
   atlas=&/CA_UI_PlayerLevel_FlameFX
   curve=&/C_UI_PlayerLevel_FlameFX_KeyArt
   group=Flame Visuals
   parameterName=FlameCurves
   sortPriority=3
  }
  Multiply(A=Min(A=@n3, B=Multiply(A=@n5, B=@n10)), B=@n12.G)
  n13 = VectorParameter() {desc=The color of the flame., group=Flame Visuals, parameterName=FlameColor, sortPriority=0, defaultValue=<rgba 0.0331 0.7011 1 1>}
  n14 = ScalarParameter() {defaultValue=0.04, desc=The opacity of the outer glow., group=Outer Glow, parameterName=OuterGlowOpacity, sliderMax=1, sortPriority=0}
  n15 = Max(A=@n14) {constB=0}
  n16 = ScalarParameter() {
   defaultValue=1
   desc=How feathered the edge of the outer glow should be.
   group=Outer Glow
   parameterName=OuterGlowSoftness
   sliderMax=1
   sortPriority=2
  }
  n17 = ScalarParameter() {
   defaultValue=0.5
   desc=How far out from the flame the outer glow should extend.
   group=Outer Glow
   parameterName=OuterGlowEdge
   sliderMax=1
   sortPriority=1
  }
  n18 = SmoothStep(Min=@n17, Max=Subtract(A=@n17, B=Max(A=@n16) {constB=0}), Value=NamedRerouteUsage()) {constMax=-0.2, constMin=0.4}
  n19 = NamedRerouteUsage()
  n20 = ScalarParameter() {defaultValue=1, desc=The opacity of the inner glow., group=Inner Glow, parameterName=InnerGlowOpacity, sliderMax=1, sortPriority=0}
  n21 = Multiply(A=Max(A=@n20) {constB=0}, B=@n3)
  n22 = Saturate(None=Add(A=Multiply(A=@n15, B=@n18), B=Min(A=@n19, B=@n21)))
  n23 = CurveAtlasRowParameter(CurveTime=ComponentMask(None=NamedRerouteUsage()) {g=true}) {
   atlas=&CA_UI_PlayerLevel_FlameFX
   curve=&C_UI_PlayerLevel_FlameFX_KeyArt
   group=Flame Visuals
   parameterName=FlameCurves
   sortPriority=3
  }
  Multiply(A=VertexColor().A, B=Multiply(A=Multiply(A=@n13.RGB, B=@n22), B=@n23.B))
  n24 = MaterialFunctionCall() {materialFunction=&GetUserInterfaceUV}
  n25 = Divide(A=ComponentMask(None=@n24.Pixel Size) {g=true}, B=ComponentMask(None=@n24.Pixel Size) {r=true})
  n26 = OneMinus(None=Divide(B=@n25) {constA=0.5})
  n27 = ComponentMask(None=@n24.Normalized UV) {g=true}
  n28 = ScalarParameter() {
   defaultValue=1
   desc="The overall height of the flame within the quad (the flame starts at the bottom)."
   group=Flame Visuals
   parameterName=FlameHeight
   sliderMax=1
   sortPriority=2
  }
  n29 = Divide(B=@n25)
  n30 = Multiply(A=@n25, B=Subtract(A=@n27, B=OneMinus(None=@n29)))
  n31 = LinearInterpolate(A=@n30, B=Multiply(A=@n27, B=Divide(B=@n26) {constA=0.5}), Alpha=@n28)
  NamedRerouteDeclaration(None=AppendVector(A=ComponentMask(None=@n24.Normalized UV) {r=true}, B=If(A=@n30, "A > B"=@n30, "A == B"=@n31, "A < B"=@n31) {constB=0.5})) {name=FlameUVs, nodeColor=<rgba 1 0 1 1>}
  n32 = ScalarParameter() {defaultValue=0.4, desc=Overall size of the flame., group=Flame Visuals, parameterName=FlameSize, sliderMax=1, sortPriority=1}
  n33 = ComponentMask(None=NamedRerouteUsage()) {g=true}
  n34 = CurveAtlasRowParameter(CurveTime=Add(A=LinearInterpolate(Alpha=@n32) {constA=-0.5, constB=0}, B=OneMinus(None=@n33))) {
   atlas=&CA_UI_PlayerLevel_FlameFX
   curve=&C_UI_PlayerLevel_FlameFX_KeyArt
   group=Flame Visuals
   parameterName=FlameCurves
   sortPriority=3
  }
  n35 = ScalarParameter() {
   defaultValue=0.4
   desc=Severity of the distortion effect.
   group=Distortion
   parameterName=DistortionAmount
   sliderMax=1
   sliderMin=-1
   sortPriority=0
  }
  n36 = NamedRerouteUsage()
  n37 = Multiply(A=@n35, B=Add(A=ComponentMask(None=@n36) {g=true, r=true}) {constB=-0.25})
  n38 = ScalarParameter() {defaultValue=0.4, desc=Overall size of the flame., group=Flame Visuals, parameterName=FlameSize, sliderMax=1, sortPriority=1}
  n39 = MaterialFunctionCall(UVs=Add(A=Multiply(A=@n34.R, B=@n37), B=NamedRerouteUsage()), Size=@n38) {materialFunction=&MF_UI_Circle}
  NamedRerouteDeclaration(None=@n39.Fill) {name=FlameFill, nodeColor=<rgba 1 0 1 1>}
  NamedRerouteDeclaration(None=@n39.SDF) {name=FlameSDF, nodeColor=<rgba 1 0 1 1>}
  n40 = ScalarParameter() {
   desc=Positional offset for flame distortion pattern. Useful for creating pattern variation when multiple flames are on screen.
   group=Distortion
   parameterName=DistortionOffsetX
   sliderMax=1
   sortPriority=3
  }
  n41 = ScalarParameter() {
   desc=Positional offset for flame distortion pattern. Useful for creating pattern variation when multiple flames are on screen.
   group=Distortion
   parameterName=DistortionOffsetY
   sliderMax=1
   sortPriority=4
  }
  n42 = AppendVector(A=@n40, B=@n41)
  n43 = MaterialFunctionCall() {materialFunction=&GetUserInterfaceUV}
  n44 = Divide(A=ComponentMask(None=@n43.Pixel Size) {g=true}, B=ComponentMask(None=@n43.Pixel Size) {r=true})
  n45 = ComponentMask(None=@n43.Normalized UV) {g=true}
  n46 = ScalarParameter() {
   defaultValue=0.8
   desc=Frequency/scale of the flame distortion pattern.
   group=Distortion
   parameterName=DIstortionPatternScaleX
   sliderMax=2
   sortPriority=2
  }
  n47 = ScalarParameter() {
   defaultValue=0.6
   desc=Frequency/scale of the flame distortion pattern.
   group=Distortion
   parameterName=DIstortionPatternScaleY
   sliderMax=2
   sortPriority=2
  }
  n48 = Multiply(A=AppendVector(A=ComponentMask(None=@n43.Normalized UV) {r=true}, B=Multiply(A=@n44, B=@n45)), B=AppendVector(A=@n46, B=@n47))
  n49 = ScalarParameter() {
   defaultValue=0.07
   desc=How fast the flame animates.
   group=Distortion
   parameterName=DistortionSpeedX
   sliderMax=2
   sliderMin=-2
   sortPriority=5
  }
  n50 = ScalarParameter() {
   defaultValue=0.8
   desc=How fast the flame animates.
   group=Distortion
   parameterName=DistortionSpeedY
   sliderMax=2
   sliderMin=-2
   sortPriority=6
  }
  n51 = TextureSampleParameter2D(UVs=Panner(Coordinate=Add(A=@n42, B=@n48), Speed=AppendVector(A=@n49, B=@n50))) {
   desc=The noise pattern texture to use for flame distortion.
   group=Distortion
   parameterName=DistortionTexture
   sortPriority=1
   texture=&T_NoisePack_01
  }
  NamedRerouteDeclaration(None=@n51.RGB) {name=DistortionTexture, nodeColor=<rgba 1 0 1 1>}
 ]
}
