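# Lobby nameplate widget for a team member pedestal, wrapping the old and new nameplate layouts in one switcher.
# On initialization, the bUseNewNameplatesAndSocialNudges runtime option picks the old banner, name and status layout or the new box around WBP_Nameplate.
# Readiness changes recolor the old layout's status text, and TryGetViewModel returns WBP_Nameplate's view model.
# In compact UI, PreConstruct sets the new box's bottom padding from safe-zone scale and aspect ratio.
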WidgetBlueprint /Game/UI/Frontend/Lobby/Nameplates/TeamMemberNameplateWrapper
graphs [
 EventGraph {nodes [
   Event PreConstruct()
   > SetCompactUIStyle() {is=title K2Node_CallFunction}
   Event OnInitialized()
   n1 = GetRuntimeOptions()
   > SetActiveWidget(
    self:Widget Switcher Object Reference=Get Switcher_OldNewNameplates() -> (Switcher_OldNewNameplates:Common Activatable Widget Switcher Object Reference)
    Widget=Select(
     Option 0:Widget Object Reference=Get FortHUDElementWrapper() -> (FortHUDElementWrapper:Fort HUDElement Wrapper Object Reference)
     Option 1:Widget Object Reference=Get NewNameplateBox() -> (NewNameplateBox:Horizontal Box Object Reference)
     Index=Get bUseNewNameplatesAndSocialNudges!(self:Fort Runtime Options Object Reference=@n1) -> (bUseNewNameplatesAndSocialNudges:bool)
    ) -> (ReturnValue:Widget Object Reference)
   )
   n2 = Event OnReadyStatusChanged!() -> (OutputDelegate, then, GameReadiness:EGameReadiness Enum=NotReady, bIsDownloading:bool) {is=title K2Node_Event}
   n3 = Get ReadyColor() -> (ReadyColor:Linear Color Structure)
   n4 = Get NotReadyColor() -> (NotReadyColor:Linear Color Structure)
   > SetColorAndOpacity(
    self=Get Text_ReadyStatus() -> (Text_ReadyStatus:Common Text Object Reference)
    InColorAndOpacity=Make Slate Color(SpecifiedColor=Select!(
      NotReady:Linear Color Structure=@n4
      Ready:Linear Color Structure=@n3
      Reevaluating:Linear Color Structure
      SittingOut:Linear Color Structure=@n4
      Away:Linear Color Structure=@n4
      Index:EGameReadiness Enum=@n2.GameReadiness
     ) -> (ReturnValue:Linear Color Structure))
   )
  ]}
 SetCompactUIStyle {nodes [
   SetCompactUIStyle() {is=title K2Node_FunctionEntry}
   > Branch(Condition=IsUsingCompactUI())
   > n1 = GetSafeZonePadding!(execute) -> (then, SafePadding:Vector 4 Structure, SafePaddingScale_X:double, SafePaddingScale_Y:double, SpillOverPadding:Vector 4 Structure)
   n2 = CompareToScreenAspectRatio!(Comparison:EAspectRatioComparison Enum=LessThan, AspectRatio:EAspectRatio Enum=Ratio_23_16) -> (ReturnValue:bool)
   n3 = float * float(A=@n1.SafePaddingScale_Y, B=20) {is=title K2Node_PromotableOperator}
   > SetPadding!(
    execute
    self:Widget Switcher Slot Object Reference=SlotAsWidgetSwitcherSlot!(Widget:Widget Object Reference=Get NewNameplateBox()) -> (ReturnValue:Widget Switcher Slot Object Reference)
    InPadding_Left:float=0
    InPadding_Top:float=0
    InPadding_Right:float=0
    InPadding_Bottom:float=float * float(A=float - float(A=1, B=@n3), B=SelectFloat(A=48, B=32, bPickA=@n2))
   ) -> (then)
  ]}
 TryGetViewModel {nodes [
   TryGetViewModel() {is=title K2Node_FunctionEntry}
   > Return Node!(
    execute
    ReturnValue:Fort Team Member Pedestal Nameplate View Model Object Reference=Get FortTeamMemberPedestalNameplateViewModel!(self:WBP Nameplate Object Reference=Get WBP_Nameplate() -> (WBP_Nameplate:WBP Nameplate Object Reference)) -> (FortTeamMemberPedestalNameplateViewModel:Fort Team Member Pedestal Nameplate View Model Object Reference)
   )
  ]}
]
variables [
 NotReadyColor {type=LinearColor, default=<rgba 1 0.1793 0.1562 1>}
 ReadyColor {type=LinearColor, default=<rgba 0.0343 1 0.227 1>}
]
widgets {
 info {
  inheritedWidgetCount=17
  namedSlotCount=2
  parentClass=/Script/FortniteUI.FortTeamMemberPedestalNameplate
  rootWidgetClass=&CommonActivatableWidgetSwitcher
  widgetCount=40
 }
 widgets [
  Switcher_OldNewNameplates {widgetClassPath=&CommonActivatableWidgetSwitcher, bIsVariable=true, widget=Switcher_OldNewNameplates}
  FortHUDElementWrapper {
   widgetClassPath=/Game/Athena/HUD/FortHUDElementWrapper
   bIsVariable=true
   parent=Switcher_OldNewNameplates
   widget=FortHUDElementWrapper
   properties {elementTag {tagName=HUD}}
   slot {class=WidgetSwitcherSlot}
  }
  Overlay_1 {widgetClassPath=&Overlay, namedSlotHost=FortHUDElementWrapper, widget=Overlay_1}
  HorizontalBox_1 {widgetClassPath=&HorizontalBox, parent=Overlay_1, widget=HorizontalBox_1, slot {class=OverlaySlot}}
  TMNW_OverlayL {
   widgetClassPath=&Overlay
   parent=HorizontalBox_1
   widget=TMNW_OverlayL
   slot {class=HorizontalBoxSlot, properties {horizontalAlignment=HAlign_Right, verticalAlignment=VAlign_Top}}
  }
  Banner_Containment {
   widgetClassPath=&Overlay
   parent=TMNW_OverlayL
   widget=Banner_Containment
   properties {renderTransformPivot=<1 0>}
   slot {class=OverlaySlot, properties {horizontalAlignment=HAlign_Right, padding=<ltrb 0 14 0 0>}}
  }
  Banner_MemberBanner {
   widgetClassPath=/Script/FortniteUI.FortPlayerBanner
   bInherited=true
   bIsVariable=true
   parent=Banner_Containment
   widget=Banner_MemberBanner
   properties {bShowBorder=false, desiredSize=<37 44>, renderTransform {angle=0, scale=<1 1>, shear=<0 0>, translation=<0 2>}}
   slot {class=OverlaySlot}
  }
  Widget_PlayerLevel {
   widgetClassPath=&SizeBox
   bIsVariable=true
   parent=Banner_Containment
   widget=Widget_PlayerLevel
   properties {bOverride_WidthOverride=true, widthOverride=34}
   slot {class=OverlaySlot, properties {verticalAlignment=VAlign_Bottom, padding=<ltrb 0 30 0 0>}}
  }
  ScaleBox_0 {
   widgetClassPath=&ScaleBox
   bIsVariable=true
   parent=Widget_PlayerLevel
   widget=ScaleBox_0
   properties {stretch=Fill}
   slot {class=SizeBoxSlot}
  }
  Text_PlayerLevel {
   widgetClassPath=&CommonTextBlock
   bInherited=true
   bIsVariable=true
   parent=ScaleBox_0
   widget=Text_PlayerLevel
   properties {
    text=888
    font {
     bForceMonospaced=false
     bMaterialIsStencil=false
     fontMaterial=null
     fontObject=&BurbankSmall
     letterSpacing=0
     monospacedWidth=1
     size=18
     skewAmount=0
     typefaceFontName=Black
     outlineSettings {
      bApplyOutlineToDropShadows=true
      bMiteredCorners=false
      bSeparateFillAlpha=false
      outlineMaterial=null
      outlineSize=1
      outlineColor=<rgba 0.0208 0.0208 0.0208 0.5>
     }
    }
    renderTransform {angle=0, scale=<1 1>, shear=<0 0>, translation=<-21 0>}
    shadowColorAndOpacity=<rgba 0 0 0 1>
    shadowOffset=<1.2 1.2>
   }
   slot {class=ScaleBoxSlot}
  }
  Image_LeaderIcon {
   widgetClassPath=&Image
   bInherited=true
   bIsVariable=true
   parent=Banner_Containment
   widget=Image_LeaderIcon
   properties {
    visibility=HitTestInvisible
    brush {
     bIsDynamicallyLoaded=false
     drawAs=Image
     imageType=NoImage
     mirroring=NoMirror
     resourceName=None
     resourceObject=/Game/UI/Foundation/Banner/M-CrownShine
     tiling=NoTile
     imageSize=<37 30>
     margin=<ltrb 0 0 0 0>
     outlineSettings {
      bUseBrushTransparency=false
      roundingType=HalfHeightRadius
      width=0
      color {colorUseRule=UseColor_Specified, specifiedColor=<rgba 0 0 0 0>}
      cornerRadii=<0 0 0 0>
     }
     tintColor {colorUseRule=UseColor_Specified, specifiedColor=<rgba 1 1 1 1>}
     uVRegion {bIsValid=false, max=<0 0>, min=<0 0>}
    }
    renderTransform {angle=0, scale=<1 1>, shear=<0 0>, translation=<0 -23>}
   }
   slot {class=OverlaySlot, properties {horizontalAlignment=HAlign_Center}}
  }
  SB_InputAction {
   widgetClassPath=&SizeBox
   bIsVariable=true
   parent=TMNW_OverlayL
   widget=SB_InputAction
   properties {visibility=Hidden}
   slot {class=OverlaySlot, properties {padding=<ltrb 0 0 0 8>}}
  }
  InputActionWidget {
   widgetClassPath=&CommonActionWidget
   bInherited=true
   bIsVariable=true
   parent=SB_InputAction
   widget=InputActionWidget
   properties {
    visibility=SelfHitTestInvisible
    inputActions [{dataTable=&InputActionDataTable, rowName=Select}]
    renderTransform {angle=0, scale=<1 1>, shear=<0 0>, translation=<-40 0>}
   }
   slot {class=SizeBoxSlot}
  }
  FortVisualAttachment_0 {
   widgetClassPath=&FortVisualAttachment
   parent=TMNW_OverlayL
   widget=FortVisualAttachment_0
   properties {widthOverride=64, contentAnchor=<0.5 0>}
   slot {class=OverlaySlot, properties {verticalAlignment=VAlign_Bottom}}
  }
  HorizontalBox_0 {
   widgetClassPath=&HorizontalBox
   parent=FortVisualAttachment_0
   widget=HorizontalBox_0
   slot {class=SizeBoxSlot, properties {horizontalAlignment=HAlign_Right}}
  }
  SizeBox_HotfixControlled {
   widgetClassPath=&SizeBox
   bInherited=true
   parent=HorizontalBox_0
   widget=SizeBox_HotfixControlled
   slot {class=HorizontalBoxSlot}
  }
  Image_PlatformIcon {
   widgetClassPath=&Image
   bInherited=true
   bIsVariable=true
   parent=SizeBox_HotfixControlled
   widget=Image_PlatformIcon
   slot {class=SizeBoxSlot}
  }
  Spacer_118 {widgetClassPath=&Spacer, parent=HorizontalBox_0, widget=Spacer_118, properties {size=<8 1>}, slot {class=HorizontalBoxSlot}}
  SizeBox_1 {
   widgetClassPath=&SizeBox
   parent=HorizontalBox_0
   widget=SizeBox_1
   properties {bOverride_HeightOverride=true, bOverride_WidthOverride=true, heightOverride=32, widthOverride=32}
   slot {class=HorizontalBoxSlot}
  }
  InputIndicator_ActiveInputType {
   widgetClassPath=/Game/Athena/UI/Matchmaking/MatchmakingInputIndicator
   bInherited=true
   bIsVariable=true
   parent=SizeBox_1
   widget=InputIndicator_ActiveInputType
   properties {bShowRemoteInputOnlyWhenDifferentFromDefault=true, inputSource=Remote}
   slot {class=SizeBoxSlot}
  }
  Spacer_Anim {
   widgetClassPath=&Spacer
   parent=TMNW_OverlayL
   widget=Spacer_Anim
   properties {size=<40 1>}
   slot {class=OverlaySlot, properties {horizontalAlignment=HAlign_Right}}
  }
  HabaneroWidget {
   widgetClassPath=/Game/UI/Frontend/Lobby/Habanero/WBP_Habanero_Icon
   bInherited=true
   bIsVariable=true
   parent=TMNW_OverlayL
   widget=HabaneroWidget
   properties {visibility=Hidden, iconMinSize=<64 64>}
   slot {class=OverlaySlot, properties {padding=<ltrb -70 0 0 0>}}
  }
  TMNW_OverlayR {
   widgetClassPath=&Overlay
   parent=HorizontalBox_1
   widget=TMNW_OverlayR
   slot {class=HorizontalBoxSlot, properties {verticalAlignment=VAlign_Top}}
  }
  VerticalBox_2 {
   widgetClassPath=&VerticalBox
   parent=TMNW_OverlayR
   widget=VerticalBox_2
   slot {class=OverlaySlot, properties {padding=<ltrb 8 0 48 0>}}
  }
  Overlay_0 {
   widgetClassPath=&Overlay
   parent=VerticalBox_2
   widget=Overlay_0
   properties {clipping=ClipToBounds}
   slot {class=VerticalBoxSlot}
  }
  Text_SocialName {
   widgetClassPath=/Script/FortniteUI.FortSocialNameTextBlock
   bInherited=true
   bIsVariable=true
   parent=Overlay_0
   widget=Text_SocialName
   properties {
    bOverrideDefaultStyle=true
    bWrapToNextLine=true
    clipping=OnDemand
    minDesiredSize=120
    scrollStyle=/FortUILibrary/Styles/Typography/Foundation/TextScrollStyle-NoFade.TextScrollStyle-NoFade_C
    text="Nickname\n<PC color=\"B5C1C1FF\" size=\"12\"></><text color=\"B5C1C1FF\" size=\"12\">Display Name</>"
    defaultTextStyleOverride {
     overflowPolicy=Clip
     transformPolicy=None
     colorAndOpacity {colorUseRule=UseColor_Specified, specifiedColor=<rgba 1 1 1 1>}
     font {
      bForceMonospaced=false
      bMaterialIsStencil=false
      fontMaterial=null
      fontObject=&HeadingNow
      letterSpacing=0
      monospacedWidth=1
      size=18
      skewAmount=0
      typefaceFontName=75 Medium
      outlineSettings {
       bApplyOutlineToDropShadows=false
       bMiteredCorners=false
       bSeparateFillAlpha=false
       outlineMaterial=null
       outlineSize=1
       outlineColor=<rgba 0 0 0 0.3>
      }
     }
     highlightColor {colorUseRule=UseColor_Specified, specifiedColor=<rgba 0 0 0 1>}
     highlightShape {
      bIsDynamicallyLoaded=false
      drawAs=Image
      imageType=NoImage
      mirroring=NoMirror
      resourceName=None
      resourceObject=null
      tiling=NoTile
      imageSize=<32 32>
      margin=<ltrb 0 0 0 0>
      outlineSettings {
       bUseBrushTransparency=false
       roundingType=HalfHeightRadius
       width=0
       color {colorUseRule=UseColor_Specified, specifiedColor=<rgba 0 0 0 0>}
       cornerRadii=<0 0 0 0>
      }
      tintColor {colorUseRule=UseColor_Specified, specifiedColor=<rgba 1 1 1 1>}
      uVRegion {bIsValid=false, max=<0 0>, min=<0 0>}
     }
     selectedBackgroundColor {colorUseRule=UseColor_Foreground, specifiedColor=<rgba 1 0 1 1>}
     shadowColorAndOpacity=<rgba 0 0 0 1>
     shadowOffset=<0 0>
     strikeBrush {
      bIsDynamicallyLoaded=false
      drawAs=Image
      imageType=NoImage
      mirroring=NoMirror
      resourceName=None
      resourceObject=null
      tiling=NoTile
      imageSize=<32 32>
      margin=<ltrb 0 0 0 0>
      outlineSettings {
       bUseBrushTransparency=false
       roundingType=HalfHeightRadius
       width=0
       color {colorUseRule=UseColor_Specified, specifiedColor=<rgba 0 0 0 0>}
       cornerRadii=<0 0 0 0>
      }
      tintColor {colorUseRule=UseColor_Specified, specifiedColor=<rgba 1 1 1 1>}
      uVRegion {bIsValid=false, max=<0 0>, min=<0 0>}
     }
     underlineBrush {
      bIsDynamicallyLoaded=false
      drawAs=Image
      imageType=NoImage
      mirroring=NoMirror
      resourceName=None
      resourceObject=null
      tiling=NoTile
      imageSize=<32 32>
      margin=<ltrb 0 0 0 0>
      outlineSettings {
       bUseBrushTransparency=false
       roundingType=HalfHeightRadius
       width=0
       color {colorUseRule=UseColor_Specified, specifiedColor=<rgba 0 0 0 0>}
       cornerRadii=<0 0 0 0>
      }
      tintColor {colorUseRule=UseColor_Specified, specifiedColor=<rgba 1 1 1 1>}
      uVRegion {bIsValid=false, max=<0 0>, min=<0 0>}
     }
    }
    secondNameColor=<rgba 0.7098 0.7569 0.7569 1>
   }
   slot {class=OverlaySlot}
  }
  HorizontalBox_3 {widgetClassPath=&HorizontalBox, parent=VerticalBox_2, widget=HorizontalBox_3, slot {class=VerticalBoxSlot}}
  Text_ReadyStatus {
   widgetClassPath=&CommonTextBlock
   bInherited=true
   bIsVariable=true
   parent=VerticalBox_2
   widget=Text_ReadyStatus
   properties {
    justification=Center
    text=Not Ready
    colorAndOpacity {colorUseRule=UseColor_Specified, specifiedColor=<rgba 1 0.1714 0.1329 1>}
    font {
     bForceMonospaced=false
     bMaterialIsStencil=false
     fontMaterial=null
     fontObject=&BurbankSmall
     letterSpacing=0
     monospacedWidth=1
     size=16
     skewAmount=0
     typefaceFontName=Bold
     outlineSettings {
      bApplyOutlineToDropShadows=false
      bMiteredCorners=false
      bSeparateFillAlpha=false
      outlineMaterial=null
      outlineSize=1
      outlineColor=<rgba 0 0 0 0.33>
     }
    }
    shadowOffset=<0 0>
   }
   slot {class=VerticalBoxSlot, properties {horizontalAlignment=HAlign_Left}}
  }
  VoiceIcon_Status {
   widgetClassPath=/Script/FortniteUI.FortVoiceChatStatusIconProxyAnim
   bInherited=true
   bIsVariable=true
   parent=TMNW_OverlayR
   widget=VoiceIcon_Status
   slot {class=OverlaySlot, properties {horizontalAlignment=HAlign_Right, verticalAlignment=VAlign_Bottom, padding=<ltrb -32 0 0 0>}}
  }
  TeamMemberPedestal_ActivatableContainer {
   widgetClassPath=/Game/UI/Frontend/Lobby/PlayerPad/TeamMemberPedestal_ActivatableContainer
   bIsVariable=true
   parent=TMNW_OverlayR
   widget=TeamMemberPedestal_ActivatableContainer
   slot {class=OverlaySlot, properties {padding=<ltrb 6 -40 0 0>}}
  }
  Button_AddFriend {
   widgetClassPath=/Game/UI/Frontend/Lobby/PlayerPad/TeamMemberPedestal_Nameplate_AddFriendButton
   bInherited=true
   bIsVariable=true
   namedSlotHost=TeamMemberPedestal_ActivatableContainer
   widget=Button_AddFriend
   properties {onButtonBaseClicked=(), onButtonBaseDoubleClicked=(), onButtonBaseHovered=(), onButtonBaseUnhovered=(), onSelectedChangedBase=()}
  }
  NewNameplateBox {
   widgetClassPath=&HorizontalBox
   bIsVariable=true
   parent=Switcher_OldNewNameplates
   widget=NewNameplateBox
   slot {class=WidgetSwitcherSlot}
  }
  VerticalBox_3 {widgetClassPath=&VerticalBox, parent=NewNameplateBox, widget=VerticalBox_3, slot {class=HorizontalBoxSlot}}
  PowerRatingWidgetSwitcher {
   widgetClassPath=&CommonWidgetSwitcherLegacy
   bInherited=true
   bIsVariable=true
   parent=VerticalBox_3
   widget=PowerRatingWidgetSwitcher
   slot {class=VerticalBoxSlot, properties {horizontalAlignment=HAlign_Center, verticalAlignment=VAlign_Bottom, size {sizeRule=Fill, value=1}}}
  }
  PowerRating_CampaignPower {
   widgetClassPath=/Game/UI/Foundation/Stats/PowerRatingWidget
   bInherited=true
   bIsVariable=true
   parent=PowerRatingWidgetSwitcher
   widget=PowerRating_CampaignPower
   properties {powerRatingType=Campaign, showDualPower=false, showProgress=false, useLargeStyle=false}
   slot {class=WidgetSwitcherSlot}
  }
  Moonbeam_PowerRating_CampaignPower {
   widgetClassPath=/Game/UI/Foundation/Stats/Moonbeam_PowerRatingWidget
   bInherited=true
   bIsVariable=true
   parent=PowerRatingWidgetSwitcher
   widget=Moonbeam_PowerRating_CampaignPower
   properties {powerRatingType=Campaign, showDualPower=false, showProgress=false, useLargeStyle=false}
   slot {class=WidgetSwitcherSlot, properties {padding=<ltrb 8 0 0 0>}}
  }
  Phoenix_PowerRating_CampaignPower {
   widgetClassPath=&Moonbeam_PowerRatingWidget
   bInherited=true
   bIsVariable=true
   parent=PowerRatingWidgetSwitcher
   widget=Phoenix_PowerRating_CampaignPower
   properties {powerRatingType=Phoenix, showDualPower=false, showProgress=false, useLargeStyle=false}
   slot {class=WidgetSwitcherSlot, properties {padding=<ltrb 8 0 0 0>}}
  }
  Image_HeroIcon {
   widgetClassPath=&Image
   bInherited=true
   bIsVariable=true
   parent=VerticalBox_3
   widget=Image_HeroIcon
   properties {
    visibility=Collapsed
    brush {
     bIsDynamicallyLoaded=false
     drawAs=Image
     imageType=NoImage
     mirroring=NoMirror
     resourceName=None
     resourceObject=/Game/UI/Foundation/Textures/Icons/ItemTypes/T-Icon-Hero-Outlander-32
     tiling=NoTile
     imageSize=<24 24>
     margin=<ltrb 0 0 0 0>
     outlineSettings {
      bUseBrushTransparency=false
      roundingType=HalfHeightRadius
      width=0
      color {colorUseRule=UseColor_Specified, specifiedColor=<rgba 0 0 0 0>}
      cornerRadii=<0 0 0 0>
     }
     tintColor {colorUseRule=UseColor_Specified, specifiedColor=<rgba 1 1 1 1>}
     uVRegion {bIsValid=false, max=<0 0>, min=<0 0>}
    }
   }
   slot {class=VerticalBoxSlot, properties {horizontalAlignment=HAlign_Center, verticalAlignment=VAlign_Bottom}}
  }
  WBP_PluginStub_Icon {
   widgetClassPath=&/WBP_PluginStub_Icon
   bIsVariable=true
   parent=NewNameplateBox
   widget=WBP_PluginStub_Icon
   slot {class=HorizontalBoxSlot, properties {horizontalAlignment=HAlign_Center, verticalAlignment=VAlign_Bottom}}
  }
  WBP_Nameplate {
   widgetClassPath=&/WBP_Nameplate
   bIsVariable=true
   parent=NewNameplateBox
   widget=WBP_Nameplate
   slot {class=HorizontalBoxSlot}
  }
 ]
}
