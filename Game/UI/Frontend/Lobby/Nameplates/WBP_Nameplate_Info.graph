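# Lobby nameplate info widget showing player names, platform icons, voice chat state and matchmaking status.
# Update functions set names, icons and voice materials, collapsing the secondary name row when the nickname is empty.
# Status updates select text and color by enum, then play an outro animation for non-empty text that holds, fades and collapses it; the intro animation is unused here.
# PreConstruct applies mobile outline materials and resized icons for compact UI.
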WidgetBlueprint /Game/UI/Frontend/Lobby/Nameplates/WBP_Nameplate_Info
animations [
 Anim_Outro_MatchmakingStatusUpdate {
  binding_count=1
  length_seconds=4
  object=&self:Anim_Outro_MatchmakingStatusUpdate
  sequence {bindings [Text_PlayerStatus {
     id=#1
     possessed_class=CommonTextBlock
     tracks [
      {
       type=MovieSceneByteTrack
       display_name=Visibility
       sections [{type=MovieSceneByteSection, end_frame=80, start_frame=0, channels [None {type=MovieSceneScriptingByteChannel, keys [[0,4],[80,1]]}]}]
      }
      {
       type=MovieSceneFloatTrack
       display_name=Render Opacity
       sections [{
         type=MovieSceneFloatSection
         end_frame=80
         start_frame=0
         channels [None {type=MovieSceneScriptingFloatChannel, keys [[0,1],[75,0.9884],[78,0]]}]
        }]
      }
      {
       type=MovieScene2DTransformTrack
       sections [{
         type=MovieScene2DTransformSection
         end_frame=80
         start_frame=0
         channels [Translation.Y {type=MovieSceneScriptingFloatChannel, keys [[0,0],[75,0],[78,-10]]}]
        }]
      }
     ]
    }]}
 }
 Anim_Intro_MatchmakingStatusUpdate {
  binding_count=1
  length_seconds=1.25
  object=&self:Anim_Intro_MatchmakingStatusUpdate
  sequence {bindings [Text_PlayerStatus {
     id=#1
     possessed_class=CommonTextBlock
     tracks [
      {
       type=MovieSceneByteTrack
       display_name=Visibility
       sections [{type=MovieSceneByteSection, end_frame=25, start_frame=0, channels [None {type=MovieSceneScriptingByteChannel, const [4,1]}]}]
      }
      {
       type=MovieSceneFloatTrack
       display_name=Render Opacity
       sections [{type=MovieSceneFloatSection, end_frame=25, start_frame=0, channels [None {type=MovieSceneScriptingFloatChannel, keys [[0,0],[20,0],[23,1]]}]}]
      }
      {
       type=MovieScene2DTransformTrack
       sections [{
         type=MovieScene2DTransformSection
         end_frame=25
         start_frame=0
         channels [Translation.Y {type=MovieSceneScriptingFloatChannel, keys [[0,-10],[20,-10],[23,0]]}]
        }]
      }
     ]
    }]}
 }
]
graphs [
 EventGraph {nodes [
   Event PreConstruct()
   > SetCompactUIStyle() {is=title K2Node_CallFunction}
   Event Construct()
   Event Tick()
  ]}
 SetCompactUIStyle {nodes [
   SetCompactUIStyle() {is=title K2Node_FunctionEntry}
   > Branch(Condition=IsUsingCompactUI())
   > SetFontOutlineMaterial!(
    execute
    self:Text Object Reference=[
     Get Text_PrimaryName() -> (Text_PrimaryName:Common Text Object Reference)
     Get Text_SecondaryName() -> (Text_SecondaryName:Common Text Object Reference)
     Get Text_PlayerStatus() -> (Text_PlayerStatus:Common Text Object Reference)
    ]
    InMaterial:Material Interface Object Reference=/FortUILibrary/MaterialInstances/Text/MI_UI_Block_Text_Outline_Mobile
   ) -> (then)
   > SetDesiredSizeOverride!(
    execute
    self:Image Object Reference=[
     Get Image_PlatformIcon_Primary() -> (Image_PlatformIcon_Primary:Image Object Reference)
     Get Image_VoiceIcon() -> (Image_VoiceIcon:Image Object Reference)
    ]
    DesiredSize:Vector 2D Structure=(X=42,Y=42)
   ) -> (then)
   > SetDesiredSizeOverride(self=Get Image_PlatformIcon_Secondary() -> (Image_PlatformIcon_Secondary:Image Object Reference), DesiredSize=(X=36,Y=36))
  ]}
 UpdateMatchmakingStatusText {nodes [
   UpdateMatchmakingStatusText!() -> (then, UserPresenceText:Text, MatchmakingStatus:EPedestalNameplateMatchmakingStatus Enum=None, DownloadProgress:int) {is=title K2Node_FunctionEntry}
   n1 = "To String (Integer)"(InInt=Get DownloadProgress() -> (DownloadProgress:int=0))
   n2 = Select!(
    None:Text
    Ready:Text="LOCTEXT(\"Ready\")"
    NotReady:Text="LOCTEXT(\"Not Ready\")"
    Playing:Text=Get UserPresenceText() -> (UserPresenceText:Text)
    Disconnected:Text="LOCTEXT(\"Disconnected\")"
    Downloading:Text="To Text (String)"(InString=Append!(A="Downloading ", B="(", C:String=@n1, D:String=%, E:String=")") -> (ReturnValue))
    Index:EPedestalNameplateMatchmakingStatus Enum=Get MatchmakingStatus() -> (MatchmakingStatus:EPedestalNameplateMatchmakingStatus Enum=None)
   ) -> (ReturnValue:Text)
   n3 = Get Text_PlayerStatus()
   > "SetText (Text)"(self=@n3, InText=@n2)
   > Set ColorAndOpacity!(
    execute
    ColorAndOpacity:Slate Color Structure=Select!(
     None_SpecifiedColor:Linear Color Structure=(R=1,G=1,B=1,A=1)
     None_ColorUseRule:ESlateColorStylingMode Enum
     Ready_SpecifiedColor:Linear Color Structure=(R=0,G=0.7835,B=0.0976,A=1)
     Ready_ColorUseRule:ESlateColorStylingMode Enum
     NotReady_SpecifiedColor:Linear Color Structure=(R=1,G=1,B=1,A=1)
     NotReady_ColorUseRule:ESlateColorStylingMode Enum
     Playing_SpecifiedColor:Linear Color Structure=(R=1,G=1,B=1,A=1)
     Playing_ColorUseRule:ESlateColorStylingMode Enum
     Disconnected_SpecifiedColor:Linear Color Structure=(R=1,G=1,B=1,A=1)
     Disconnected_ColorUseRule:ESlateColorStylingMode Enum
     Downloading_SpecifiedColor:Linear Color Structure=(R=1,G=1,B=1,A=1)
     Downloading_ColorUseRule:ESlateColorStylingMode Enum
     Index=Get MatchmakingStatus() -> (MatchmakingStatus=None)
    ) -> (ReturnValue:Slate Color Structure)
    self:Text Object Reference=@n3
   ) -> (then, Output_Get:Slate Color Structure)
   > n4 = Branch(Condition=TextIsEmpty(InText=@n2))
   PlayAnimationForward(
    execute=@n4.else
    InAnimation=Get Anim_Outro_MatchmakingStatusUpdate() -> (Anim_Outro_MatchmakingStatusUpdate:Widget Animation Object Reference)
   )
  ]}
 UpdatePlatformIcon {nodes [
   n1 = UpdatePlatformIcon!() -> (then, PlatformIconTag_DisplayName:String, PlatformIconTag_Nickname:String, PlatformIconTag_PlatformIconTag:String) {is=title K2Node_FunctionEntry}
   > SetPlatformPrefixIcon!(execute, Image:Image Object Reference=Get Image_PlatformIcon_Secondary(), OtherPlayerPlatform:String=@n1.PlatformIconTag_PlatformIconTag) -> (then)
   > n2 = Branch(Condition=IsEmpty(InString=@n1.PlatformIconTag_Nickname))
   n3 = Get Image_PlatformIcon_Primary()
   > SetPlatformPrefixIcon(Image=@n3, OtherPlayerPlatform=@n1.PlatformIconTag_PlatformIconTag)
   SetPlatformPrefixIcon(execute=@n2.else, Image=@n3)
  ]}
 UpdatePrimaryName {nodes [
   n1 = UpdatePrimaryName!() -> (then, PrimaryName:String) {is=title K2Node_FunctionEntry}
   > Set Text(Text="To Text (String)"(InString=@n1), self=Get Text_PrimaryName())
  ]}
 UpdateSecondaryInfoVisibility {nodes [
   n1 = UpdateSecondaryInfoVisibility!() -> (then, Nickname:String) {is=title K2Node_FunctionEntry}
   > SetVisibility(
    self=Get HBox_SecondaryNameAndPlatformIcon() -> (HBox_SecondaryNameAndPlatformIcon:Horizontal Box Object Reference)
    InVisibility=Select!(Option 0:ESlateVisibility Enum=SelfHitTestInvisible, Option 1:ESlateVisibility Enum=Collapsed, Index:bool=IsEmpty(InString=@n1)) -> (ReturnValue:ESlateVisibility Enum)
   )
  ]}
 UpdateSecondaryName {nodes [
   n1 = UpdateSecondaryName!() -> (then, SecondaryName:String) {is=title K2Node_FunctionEntry}
   > Set Text(Text="To Text (String)"(InString=@n1), self=Get Text_SecondaryName())
  ]}
 UpdateVoiceChatIconState {nodes [
   n1 = UpdateVoiceChatIconState!() -> (then, VoiceState:ENameplateVoiceState Enum=NewEnumerator0) {is=title K2Node_FunctionEntry}
   > SetBrushFromMaterial(
    self=Get Image_VoiceIcon()
    Material=Select!(
     NewEnumerator0:Material Interface Object Reference=Get Icon_Voice_Idle() -> (Icon_Voice_Idle:Material Instance Object Reference)
     NewEnumerator1:Material Interface Object Reference=Get Icon_Voice_Talking() -> (Icon_Voice_Talking:Material Instance Object Reference)
     NewEnumerator2:Material Interface Object Reference=Get Icon_Voice_Muted() -> (Icon_Voice_Muted:Material Instance Object Reference)
     Index:ENameplateVoiceState Enum=@n1
    ) -> (ReturnValue:Material Interface Object Reference)
   )
  ]}
 UpdateVoiceChatIconVisibility {nodes [
   n1 = UpdateVoiceChatIconVisibility!() -> (then, IconVisibility:ESlateVisibility Enum=Visible) {is=title K2Node_FunctionEntry}
   > Set Visibility(Visibility=@n1, self=Get Image_VoiceIcon())
  ]}
]
variables [
 Icon_Voice_Idle {type=MaterialInstance, default=/FortUILibrary/Textures/MaterialInstances/MI_UIKit_Icon_VoiceChatInactive}
 Icon_Voice_Muted {type=MaterialInstance, default=/FortUILibrary/Textures/MaterialInstances/MI_UIKit_Icon_VoiceChatMuted}
 Icon_Voice_Talking {type=MaterialInstance, default=/FortUILibrary/Textures/MaterialInstances/MI_UIKit_Icon_VoiceChatActive}
]
widgets {
 info {inheritedWidgetCount=0, parentClass=&CommonActivatableWidget, rootWidgetClass=&VerticalBox, widgetCount=11}
 widgets [
  VerticalBox_1 {widgetClassPath=&VerticalBox, widget=VerticalBox_1}
  HBox_PrimaryInfo {
   widgetClassPath=&HorizontalBox
   parent=VerticalBox_1
   widget=HBox_PrimaryInfo
   slot {class=VerticalBoxSlot, properties {horizontalAlignment=HAlign_Left}}
  }
  Text_PrimaryName {
   widgetClassPath=&CommonTextBlock
   bIsVariable=true
   parent=HBox_PrimaryInfo
   widget=Text_PrimaryName
   properties {
    applyLineHeightToBottomLine=false
    mobileFontSizeMultiplier=1.5
    text="[PH] Name"
    font {
     bForceMonospaced=false
     bMaterialIsStencil=false
     fontMaterial=null
     fontObject=&HeadingNow
     letterSpacing=0
     monospacedWidth=1
     size=15.75
     skewAmount=0
     typefaceFontName=75 Medium
     outlineSettings {
      bApplyOutlineToDropShadows=false
      bMiteredCorners=false
      bSeparateFillAlpha=false
      outlineMaterial=/Game/UI/Frontend/Lobby/Nameplates/Materials/MI_UI_Nameplate_TextOutline
      outlineSize=2
      outlineColor=<rgba 0 0 0 1>
     }
    }
    shadowOffset=<0 0>
   }
   slot {class=HorizontalBoxSlot, properties {horizontalAlignment=HAlign_Left, verticalAlignment=VAlign_Center, padding=<ltrb 0 0 8 0>}}
  }
  Image_PlatformIcon_Primary {
   widgetClassPath=&Image
   bIsVariable=true
   parent=HBox_PrimaryInfo
   widget=Image_PlatformIcon_Primary
   properties {brush {
     bIsDynamicallyLoaded=false
     drawAs=Image
     imageType=NoImage
     mirroring=NoMirror
     resourceName=None
     resourceObject=null
     tiling=NoTile
     imageSize=<24 24>
     margin=<ltrb 0 0 0 0>
     outlineSettings {
      bUseBrushTransparency=false
      roundingType=HalfHeightRadius
      width=0
      color {colorUseRule=UseColor_Specified, specifiedColor=<rgba 0 0 0 0>}
      cornerRadii=<0 0 0 0>
     }
     tintColor {colorUseRule=UseColor_Specified, specifiedColor=<rgba 1 1 1 1>}
     uVRegion {bIsValid=false, max=<0 0>, min=<0 0>}
    }}
   slot {class=HorizontalBoxSlot, properties {horizontalAlignment=HAlign_Center, verticalAlignment=VAlign_Center}}
  }
  Image_VoiceIcon {
   widgetClassPath=&Image
   bIsVariable=true
   parent=HBox_PrimaryInfo
   widget=Image_VoiceIcon
   properties {
    visibility=SelfHitTestInvisible
    brush {
     bIsDynamicallyLoaded=false
     drawAs=Image
     imageType=FullColor
     mirroring=NoMirror
     resourceName=None
     resourceObject=&MI_UIKit_Icon_VoiceChatMuted
     tiling=NoTile
     imageSize=<24 24>
     margin=<ltrb 0 0 0 0>
     outlineSettings {
      bUseBrushTransparency=false
      roundingType=HalfHeightRadius
      width=0
      color {colorUseRule=UseColor_Specified, specifiedColor=<rgba 0 0 0 0>}
      cornerRadii=<0 0 0 0>
     }
     tintColor {colorUseRule=UseColor_Specified, specifiedColor=<rgba 1 1 1 1>}
     uVRegion {bIsValid=false, max=<0 0>, min=<0 0>}
    }
   }
   slot {class=HorizontalBoxSlot, properties {horizontalAlignment=HAlign_Center, verticalAlignment=VAlign_Center, padding=<ltrb 4 0 0 0>}}
  }
  VBox_SecondaryInfo {
   widgetClassPath=&VerticalBox
   bIsVariable=true
   parent=VerticalBox_1
   widget=VBox_SecondaryInfo
   slot {class=VerticalBoxSlot, properties {horizontalAlignment=HAlign_Left}}
  }
  HBox_SecondaryNameAndPlatformIcon {
   widgetClassPath=&HorizontalBox
   bIsVariable=true
   parent=VBox_SecondaryInfo
   widget=HBox_SecondaryNameAndPlatformIcon
   slot {class=VerticalBoxSlot}
  }
  Text_SecondaryName {
   widgetClassPath=&CommonTextBlock
   bIsVariable=true
   parent=HBox_SecondaryNameAndPlatformIcon
   widget=Text_SecondaryName
   properties {
    applyLineHeightToBottomLine=false
    mobileFontSizeMultiplier=1.6
    text="[PH] SecondaryName"
    font {
     bForceMonospaced=false
     bMaterialIsStencil=false
     fontMaterial=null
     fontObject=&HeadingNow
     letterSpacing=10
     monospacedWidth=1
     size=11.25
     skewAmount=0
     typefaceFontName=74 Regular
     outlineSettings {
      bApplyOutlineToDropShadows=false
      bMiteredCorners=false
      bSeparateFillAlpha=false
      outlineMaterial=&MI_UI_Nameplate_TextOutline
      outlineSize=2
      outlineColor=<rgba 0 0 0 1>
     }
    }
    shadowOffset=<0 0>
   }
   slot {class=HorizontalBoxSlot, properties {horizontalAlignment=HAlign_Left, verticalAlignment=VAlign_Center, padding=<ltrb 0 0 8 0>}}
  }
  Image_PlatformIcon_Secondary {
   widgetClassPath=&Image
   bIsVariable=true
   parent=HBox_SecondaryNameAndPlatformIcon
   widget=Image_PlatformIcon_Secondary
   properties {brush {
     bIsDynamicallyLoaded=false
     drawAs=Image
     imageType=NoImage
     mirroring=NoMirror
     resourceName=None
     resourceObject=null
     tiling=NoTile
     imageSize=<24 24>
     margin=<ltrb 0 0 0 0>
     outlineSettings {
      bUseBrushTransparency=false
      roundingType=HalfHeightRadius
      width=0
      color {colorUseRule=UseColor_Specified, specifiedColor=<rgba 0 0 0 0>}
      cornerRadii=<0 0 0 0>
     }
     tintColor {colorUseRule=UseColor_Specified, specifiedColor=<rgba 1 1 1 1>}
     uVRegion {bIsValid=false, max=<0 0>, min=<0 0>}
    }}
   slot {class=HorizontalBoxSlot, properties {horizontalAlignment=HAlign_Center, verticalAlignment=VAlign_Center}}
  }
  FortVisualAttachment_PlayerStatus {
   widgetClassPath=&FortVisualAttachment
   parent=VBox_SecondaryInfo
   widget=FortVisualAttachment_PlayerStatus
   slot {class=VerticalBoxSlot}
  }
  Text_PlayerStatus {
   widgetClassPath=&CommonTextBlock
   bIsVariable=true
   parent=FortVisualAttachment_PlayerStatus
   widget=Text_PlayerStatus
   properties {
    applyLineHeightToBottomLine=false
    mobileFontSizeMultiplier=1.6
    text="[PH] Player Status"
    visibility=Collapsed
    font {
     bForceMonospaced=false
     bMaterialIsStencil=false
     fontMaterial=null
     fontObject=&HeadingNow
     letterSpacing=10
     monospacedWidth=1
     size=11.25
     skewAmount=0
     typefaceFontName=74 Regular
     outlineSettings {
      bApplyOutlineToDropShadows=false
      bMiteredCorners=false
      bSeparateFillAlpha=false
      outlineMaterial=&MI_UI_Nameplate_TextOutline
      outlineSize=2
      outlineColor=<rgba 0 0 0 1>
     }
    }
    shadowOffset=<0 0>
   }
   slot {class=SizeBoxSlot}
  }
 ]
}
