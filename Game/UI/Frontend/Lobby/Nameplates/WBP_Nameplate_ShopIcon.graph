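# Lobby nameplate widget that shows a shop icon centered over a material-driven background shape.
# PlayIntroAnimation, not called within the widget, plays Intro, which slides and pops both images in horizontally from a zero-width background.
# Intro also flashes both from white, settling the background material to yellow-green and the icon to near black.
# PreConstruct calls SetCompactUIStyle, which overrides both images' desired sizes to larger values when compact UI is active.
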WidgetBlueprint /Game/UI/Frontend/Lobby/Nameplates/WBP_Nameplate_ShopIcon
animations [Intro {
  binding_count=2
  length_seconds=1.3
  object=&self:Intro
  sequence {bindings [
    Background {
     id=#1
     possessed_class=Image
     tracks [
      {
       type=MovieSceneWidgetMaterialTrack
       sections [{
         type=MovieSceneParameterSection
         end_frame=26
         start_frame=0
         channels [
          CircleSize {type=MovieSceneScriptingFloatChannel, default=0.95, const [0.9,2]}
          ColorR {type=MovieSceneScriptingFloatChannel, default=0.9301, thinned_from=5, keys [[7,1],[10,1],[12,0.9301],[21,0.9301]]}
          ColorG {type=MovieSceneScriptingFloatChannel, default=1, const [1,5]}
          ColorB {type=MovieSceneScriptingFloatChannel, default=0.0103, thinned_from=5, keys [[7,1],[10,1],[12,0.0103],[21,0.0103]]}
          ColorA {type=MovieSceneScriptingFloatChannel, default=1, const [1,5]}
         ]
        }]
      }
      {
       type=MovieScene2DTransformTrack
       sections [{
         type=MovieScene2DTransformSection
         end_frame=16
         start_frame=0
         channels [
          Translation.X {type=MovieSceneScriptingFloatChannel, keys [[7,-54.9252],[8,-68.9974],[12,1]]}
          Scale.X {type=MovieSceneScriptingFloatChannel, keys [[6,0],[7,1],[8,1],[10,1.6019],[11,0.7791],[12,1]]}
         ]
        }]
      }
      {
       type=MovieSceneColorTrack
       display_name=Color and Opacity
       sections [{
         type=MovieSceneColorSection
         end_frame=14
         start_frame=0
         channels [
          Color.R {type=MovieSceneScriptingFloatChannel, default=1}
          Color.G {type=MovieSceneScriptingFloatChannel, default=1}
          Color.B {type=MovieSceneScriptingFloatChannel, default=1}
          Color.A {type=MovieSceneScriptingFloatChannel, const [1,2]}
         ]
        }]
      }
     ]
    }
    ShopIcon {
     id=#2
     possessed_class=Image
     tracks [
      {
       type=MovieScene2DTransformTrack
       sections [{
         type=MovieScene2DTransformSection
         end_frame=20
         start_frame=0
         channels [
          Translation.X {type=MovieSceneScriptingFloatChannel, default=0, keys [[8.41,-99.9443],[9.47,-110.7598],[13.6267,1],[16.7443,1]]}
          Scale.X {
           type=MovieSceneScriptingFloatChannel
           default=0.104
           keys [[7.3917,0],[8.41,1],[8.431,1],[11.5483,1.4194],[12.5877,1.8538],[13.6267,0.71],[14.666,1]]
          }
         ]
        }]
      }
      {
       type=MovieSceneColorTrack
       display_name=Color and Opacity
       sections [{
         type=MovieSceneColorSection
         end_frame=26
         start_frame=0
         channels [
          Color.R {type=MovieSceneScriptingFloatChannel, default=0.0048, thinned_from=5, keys [[8.431,1],[11.5483,1],[14.666,0.0048],[21.9607,0.0048]]}
          Color.G {type=MovieSceneScriptingFloatChannel, default=0.004, thinned_from=5, keys [[8.431,1],[11.5483,1],[14.666,0.004],[21.9607,0.004]]}
          Color.B {type=MovieSceneScriptingFloatChannel, default=0.0116, thinned_from=5, keys [[8.431,1],[11.5483,1],[14.666,0.0116],[21.9607,0.0116]]}
          Color.A {type=MovieSceneScriptingFloatChannel, default=1}
         ]
        }]
      }
     ]
    }
   ]}
 }]
graphs [
 EventGraph {nodes [
   Event PreConstruct()
   > SetCompactUIStyle() {is=title K2Node_CallFunction}
  ]}
 PlayIntroAnimation {nodes [
   PlayIntroAnimation() {is=title K2Node_FunctionEntry}
   > PlayAnimation(InAnimation=Get Intro())
  ]}
 SetCompactUIStyle {nodes [
   SetCompactUIStyle() {is=title K2Node_FunctionEntry}
   > Branch(Condition=IsUsingCompactUI())
   > SetDesiredSizeOverride!(execute, self:Image Object Reference=Get Background() -> (Background:Image Object Reference), DesiredSize:Vector 2D Structure=(X=56,Y=56)) -> (then)
   > SetDesiredSizeOverride(self=Get ShopIcon() -> (ShopIcon:Image Object Reference), DesiredSize=(X=28,Y=28))
  ]}
]
variables []
widgets {
 info {inheritedWidgetCount=0, parentClass=&UserWidget, rootWidgetClass=&Overlay, widgetCount=3}
 widgets [
  Overlay_14 {widgetClassPath=&Overlay, widget=Overlay_14}
  Background {
   widgetClassPath=&Image
   bIsVariable=true
   parent=Overlay_14
   widget=Background
   properties {
    brush {
     bIsDynamicallyLoaded=false
     drawAs=Image
     imageType=FullColor
     mirroring=NoMirror
     resourceName=None
     resourceObject=/Game/UI/Frontend/Lobby/Nameplates/Materials/MI_Nameplate_BaseShape
     tiling=NoTile
     imageSize=<32 32>
     margin=<ltrb 0 0 0 0>
     outlineSettings {
      bUseBrushTransparency=false
      roundingType=HalfHeightRadius
      width=0
      color {colorUseRule=UseColor_Specified, specifiedColor=<rgba 0 0 0 0>}
      cornerRadii=<0 0 0 0>
     }
     tintColor {colorUseRule=UseColor_Specified, specifiedColor=<rgba 1 1 1 1>}
     uVRegion {bIsValid=false, max=<0 0>, min=<0 0>}
    }
    renderTransform {angle=0, scale=<0 1>, shear=<0 0>, translation=<0 0>}
   }
   slot {class=OverlaySlot, properties {horizontalAlignment=HAlign_Fill, verticalAlignment=VAlign_Fill}}
  }
  ShopIcon {
   widgetClassPath=&Image
   bIsVariable=true
   parent=Overlay_14
   widget=ShopIcon
   properties {
    brush {
     bIsDynamicallyLoaded=false
     drawAs=Image
     imageType=FullColor
     mirroring=NoMirror
     resourceName=None
     resourceObject=/FortUILibrary/Textures/MaterialInstances/MI_UIKit_Icon_Shop
     tiling=NoTile
     imageSize=<20 20>
     margin=<ltrb 0 0 0 0>
     outlineSettings {
      bUseBrushTransparency=false
      roundingType=HalfHeightRadius
      width=0
      color {colorUseRule=UseColor_Specified, specifiedColor=<rgba 0 0 0 0>}
      cornerRadii=<0 0 0 0>
     }
     tintColor {colorUseRule=UseColor_Specified, specifiedColor=<rgba 1 1 1 1>}
     uVRegion {bIsValid=false, max=<0 0>, min=<0 0>}
    }
    colorAndOpacity=<rgba 0.0048 0.004 0.0116 1>
   }
   slot {class=OverlaySlot, properties {horizontalAlignment=HAlign_Center, verticalAlignment=VAlign_Center}}
  }
 ]
}
