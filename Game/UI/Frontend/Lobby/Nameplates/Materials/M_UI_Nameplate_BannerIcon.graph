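# Translucent UI material for a lobby nameplate banner icon, drawing a tintable icon texture with an adjustable vertical offset.
# The scaled icon's red channel serves as alpha, passes through edge-pixel clipping before tinting, and is multiplied by a rerouted mask.
# A hexagon SDF provides fill and outline masks, using UVs rotated a fixed quarter turn and scaled by distance from the top and bottom edges.
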Material /Game/UI/Frontend/Lobby/Nameplates/Materials/M_UI_Nameplate_BannerIcon
graph {
 properties {
  bUsedWithStaticMesh=false
  blendMode=BLEND_Translucent
  materialDomain=MD_UI
  pixelDepthOffsetMode=PDOM_Legacy
  usedShadingModels=Default Lit
 }
 nodes [
  n1 = Add(
   A=MaterialFunctionCall(
    UVs=MaterialFunctionCall() {materialFunction=&GetUserInterfaceUV}.Normalized UV
    Texture Scale=Constant() {r=0.95}
   ) {materialFunction=&ScaleUVsByCenter}.UVs
   B=AppendVector(A=Constant(), B=ScalarParameter() {defaultValue=-0.06, parameterName=Y Offset})
  )
  n2 = TextureSample(UVs=@n1, Tex=TextureObjectParameter() {parameterName=Icon Texture, texture=/Game/Athena/HUD/MiniMap/Art/T_UI_Icon_Tornado}) {samplerSource=SSM_Clamp_WorldGroupSettings}
  Multiply(A=@n2.R, B=NamedRerouteUsage())
  Multiply(
   A=VectorParameter() {parameterName=Icon Color, defaultValue=<rgba 1 1 1 1>}.RGB
   B=MaterialFunctionCall!(UVs=@n1, Alpha=@n2.R) {materialFunction=/FortUILibrary/MaterialFunctions/Basics/MF_UI_ClipEdgePixel}.Result
  )
  n3 = MaterialFunctionCall!(
   UV=NamedRerouteUsage()
   Radius=Constant() {r=0.9}
   Corner Radius=ScalarParameter() {defaultValue=0.1, parameterName=Corner Radius}
   Outline Thickness=ScalarParameter() {defaultValue=5, parameterName=Outline Thickness}
  ) {materialFunction=/FortUILibrary/MaterialFunctions/Shapes/MF_UI_HexagonSDF}
  NamedRerouteDeclaration(None=@n3.Fill) {name=hexagonFill, nodeColor=<rgba 0.5234 0 1 1>}
  NamedRerouteDeclaration(None=@n3.Outline) {name=hexagonOutline, nodeColor=<rgba 0 0.5547 1 1>}
  n4 = MaterialFunctionCall() {materialFunction=&GetUserInterfaceUV}
  n5 = ComponentMask(None=@n4.Normalized UV) {g=true}
  n6 = Min(A=OneMinus(None=@n5), B=@n5)
  n7 = Multiply(
   A=LinearInterpolate(A=Constant2Vector() {g=0.95, r=1}, B=Constant2Vector() {g=1.05, r=1}, Alpha=Multiply(A=@n6) {constB=2}) {constA=1}
   B=Constant2Vector() {g=1.025, r=1.1}
  )
  NamedRerouteDeclaration(None=Rotator(
    Coordinate=MaterialFunctionCall(UVs=@n4.Normalized UV, Texture Scale=@n7) {materialFunction=&ScaleUVsByCenter}.UVs
    Time=Constant() {r=0.25}
   ) {speed=6.2832}) {name=hexagonUV, nodeColor=<rgba 1 0.7031 0 1>}
 ]
}
